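# Niagara system for the build-up flare of the player's Sahara emote, pairing a ground glow with rising sparks, both tinted by User.Color.
# GroundAirglow bursts two upward-facing sprites, split across renderers by spawn group, plus a light; they brighten, flicker through stepped curve keys, then fade.
# Sparks spawn at a rate that climbs over the loop then cuts off, leaving a flattened ring upward and outward under noise and drag.
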NiagaraSystem /BRCosmetics/Animation/Game/MainPlayer/Emotes/Sahara/FX/NS_Emote_Sahara_BuildUp_Flare
dependencies {
 usedDataInterfaces [&NiagaraDataInterfaceCurve, &NiagaraDataInterfaceVector2DCurve]
 usedDynamicInputs []
 usedModules [&SampleSkeletalMesh, &SystemState]
}
emitters {
 GroundAirglow {
  emitterUpdateScript [
   &EmitterState()
   &SpawnBurst_Instantaneous(Spawn Count=2, Use Spawn Probability=false)
  ]
  particleSpawnScript [
   &FN_SetSpawnGroup(Particles Per Group=1, Pattern=::NewEnumerator0)
   /Niagara/Modules/Spawn/Initialization/InitializeParticle(
    Color=&MultiplyLinearColor(Linear Color=$User.Color, Scale Factor=&MakeVector4_FromFloat(Float=0.65))
    Lifetime=1.25
    Module.Write Color=true
    Module.Write Lifetime=true
    Module.Write Mass=true
    Module.Write Material Random=true
    Module.Write Position=false
    Module.Write Scale=false
    Module.Write SpriteRotation=false
    Module.Write SpriteSize=true
    Position=&SimulationPosition()
    Sprite Size=&Vector2DFromFloat(Value=&Multiply_Float(A=200, B=&MaskFloatBySpawnGroup(Masked Float=1, Passthrough Float=0.8)))
   )
   &SpriteFacingAndAlignment(Sprite Facing=<0 0 1>, WriteAlignment=false, WriteFacing=true)
   /SetVariables(Particles.VisibilityTag=$Particles.SpawnGroup) {isSetParameters=true}
   &CameraOffset(Camera Offset Amount=&MaskFloatBySpawnGroup(Masked Float=1, Passthrough Float=100))
  ]
  particleUpdateScript [
   &ParticleState()
   &SolveForcesAndVelocity(Clamp Velocity=false, Limit Acceleration=false)
   &ScaleColor(
    Scale Alpha=&FloatFromCurve(FloatCurve=DataInterface() {propertyValues {Curve {keys [
         {arriveTangent=0.1757, interpMode=RCIM_Cubic, leaveTangent=0.1757, tangentMode=RCTM_User, time=0, value=0}
         {arriveTangent=-0.2874, interpMode=RCIM_Cubic, leaveTangent=-0.2874, tangentMode=RCTM_User, time=0.8, value=1}
         {arriveTangent=-0.0405, interpMode=RCIM_Cubic, leaveTangent=-0.0405, tangentMode=RCTM_User, time=1, value=0}
        ]}}})
    Scale RGB=&VectorFromFloat(Value=&Multiply_Float(
      A=&MaskFloatBySpawnGroup(Masked Float=1.5, Passthrough Float=1)
      B=&FloatFromCurve(FloatCurve=DataInterface() {propertyValues {Curve {keys [<at 0.7 1 RCIM_Constant>,<at 0.75 -1 RCIM_Constant>,<at 0.8 2 RCIM_Constant>,<at 0.9 -1 RCIM_Constant>,<at 0.92 1 RCIM_Constant>,<at 0.95 0 RCIM_Constant>,<at 0.98 1>]}}})
     ))
    ScaleA=true
    ScaleRGB=true
   )
   &FN_Light_Attributes(
    Color=$Particles.Color
    Create Color Attribute=true
    Create Exponent Attribute=true
    Create Radius Attribute=true
    Exponent=1.5
    Radius=150
    Scale Color=&FloatFromCurve(
     FloatCurve=DataInterface() {propertyValues {Curve {keys [
         {arriveTangent=0.2526, interpMode=RCIM_Cubic, leaveTangent=0.2526, tangentMode=RCTM_User, time=0, value=0}
         {arriveTangent=1.8091, interpMode=RCIM_Cubic, leaveTangent=1.8091, tangentMode=RCTM_User, time=0.8, value=1}
         {arriveTangent=-0.0405, interpMode=RCIM_Cubic, leaveTangent=-0.0405, tangentMode=RCTM_User, time=1, value=0}
        ]}}}
     Scale Curve=5
    )
   )
  ]
  rendererClasses [&NiagaraSpriteRendererProperties, &NiagaraLightRendererProperties]
  renderers [
   {rendererClass=&NiagaraSpriteRendererProperties}
   {rendererClass=&NiagaraSpriteRendererProperties, rendererIndex=1}
   {rendererClass=&NiagaraLightRendererProperties, rendererIndex=2}
  ]
 }
 Sparks {
  emitterUpdateScript [
   &EmitterState()
   &SpawnRate(
    SpawnRate=&FloatFromCurve(
     FloatCurve=DataInterface() {propertyValues {Curve {keys [
         {arriveTangent=0.0145, interpMode=RCIM_Cubic, leaveTangent=0.0145, tangentMode=RCTM_User, time=0, value=0}
         {arriveTangent=2.5625, interpMode=RCIM_Cubic, leaveTangent=2.5625, tangentMode=RCTM_User, time=0.8, value=1}
         {interpMode=RCIM_Cubic, leaveTangent=1.25, time=0.8373, value=0}
        ]}}}
     CurveIndex=$Emitter.NormalizedLoopAge
     Scale Curve=80
    )
    Use Spawn Probability=false
   )
  ]
  particleSpawnScript [
   &InitializeParticle(
    Color=&MultiplyLinearColor(Linear Color=$User.Color, Scale Factor=&MakeVector4_FromFloat(Float=30))
    Lifetime=&UniformRangedFloat(Minimum=0.2, Maximum=0.5)
    Mesh Scale=<1.5 5 1>
    Module.Write Color=true
    Module.Write Lifetime=true
    Module.Write Mass=true
    Module.Write Material Random=true
    Module.Write Position=false
    Module.Write Scale=false
    Module.Write SpriteRotation=false
    Module.Write SpriteSize=true
    Position=&SimulationPosition()
    Sprite Size=&Multiply_Vector2D_ByFloat(
     Float=&MaskFloatBySpawnGroup(Masked Float=0.4, Passthrough Float=0.6, Spawn Group Mask=1)
     Vector2D=&UniformRangedVector2D(Minimum=<1 2.5>, Maximum=<1 3.5>)
    )
   )
   /Niagara/Modules/Spawn/Location/SphereLocation(MaskSpawnGroup=false, Module.Override Randomness=false, Non Uniform Scale=<2.5 2.5 0>, Sphere Radius=10, Surface Only=true)
   &AddVelocity(Velocity=&RandomRangeVector(Minimum=<0 0 150>, Maximum=<0 0 250>)) {module=AddVelocity001}
   &AddVelocityFromPoint(
    Module.Use Falloff Distance=false
    Velocity Strength=&Multiply_Float(
     A=&MaskFloatBySpawnGroup(Masked Float=25, Passthrough Float=50, Spawn Group Mask=1)
     B=&RandomRangeFloat(Minimum=0.8, Maximum=1.2)
    )
   )
  ]
  particleUpdateScript [
   &ParticleState()
   &VectorNoiseForce(Force Amount=1500, Module.Use Random Noise Time Interval=false)
   &Drag(Drag=5, Use Linear Drag=true, Use Rotational Drag=true)
   &SolveForcesAndVelocity(Clamp Velocity=false, Limit Acceleration=false)
   &ScaleColor(
    Scale Alpha=&FloatFromCurve(FloatCurve=DataInterface() {propertyValues {Curve {keys [
         {arriveTangent=-2.7664, interpMode=RCIM_Cubic, leaveTangent=-2.7664, tangentMode=RCTM_User, time=0, value=0.5}
         {arriveTangent=2, interpMode=RCIM_Cubic, leaveTangent=2, tangentMode=RCTM_Break, time=0.2, value=0.2621}
         {arriveTangent=-0.3277, interpMode=RCIM_Cubic, leaveTangent=-3.899, tangentMode=RCTM_Break, time=0.25, value=1}
         <at 1 0>
        ]}}})
    ScaleA=true
    ScaleRGB=true
   )
   &DynamicMaterialParameters(
    Index 0 Param 1=&FloatFromCurve(FloatCurve=DataInterface() {propertyValues {Curve {keys [<at 0.25 0.7>,<at 0.5 0.05>]}}})
    Index 0 Param 2=2
    Index 0 Param 3=1
    Index 0 Param 4=1
    Module.Param0WriteEnabled=true
    Module.Param1WriteEnabled=true
    Module.Param2WriteEnabled=true
    Module.Param3WriteEnabled=true
   )
   &CameraOffset(Camera Offset Amount=50)
   &ScaleSpriteSize(Scale Factor=&Vector2DFromCurve(
     Vector2Curve=DataInterface() {propertyValues {XCurve {keys [<at 0 1.5>,<at 0.3 1.262>,<at 1 1>]}, YCurve {keys [<at 0 1>,<at 0.3 1.1>,<at 1 0>]}}}
     ScaleCurve=<2 2>
    ))
   &ScaleSpriteSize(Scale Factor=<0.4 1>, Scale Sprite Size Mode=::NewEnumerator2) {module=ScaleSpriteSize001}
  ]
  rendererClasses [&NiagaraSpriteRendererProperties]
  renderers [{rendererClass=&NiagaraSpriteRendererProperties}]
 }
}
summary {
 systemName=NS_Emote_Sahara_BuildUp_Flare
 emitters [
  {emitterName=Sparks, rendererClasses [&NiagaraSpriteRendererProperties]}
  {emitterName=GroundAirglow, rendererClasses [&NiagaraSpriteRendererProperties, &NiagaraLightRendererProperties]}
 ]
 userVariables [
  User.bIsInWater {type=NiagaraBool}
  User.Color {type=LinearColor, description="", defaultValue {struct=LinearColor, value=<rgba 0.25 0.6 1 1>}}
  User.SkeletalMesh {
   type=NiagaraDataInterfaceSkeletalMesh
   defaultValue {
    struct=NiagaraExt_VariableValue_DataInterface
    value {dataInterface=&self:NiagaraDataInterfaceSkeletalMesh_0}
   }
  }
 ]
}
