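# Niagara system for the Sahara emote that flashes brief, random lightning on a character's arms and hands, tinted by User.Color.
# Mesh and Spline emitter pairs sample separate bone/socket indices of User.SkeletalMesh, each combining mesh, sprite and light renderers.
# Mesh arcs spawn on bone-oriented cylinder surfaces and re-roll their spin every update to flicker.
# Spline bolts follow the bone with random mesh variants, using probability-gated bursts for irregular strikes.
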NiagaraSystem /BRCosmetics/Animation/Game/MainPlayer/Emotes/Sahara/FX/NS_Emote_Sahara_RandomLightning
dependencies {usedDataInterfaces [&NiagaraDataInterfaceCurve], usedDynamicInputs [], usedModules [&SystemState]}
emitters {
 LightningMeshArms {
  emitterUpdateScript [
   &EmitterState()
   &SpawnBurst_Instantaneous(Use Spawn Probability=false, UseLoopCountLimit=false)
   &SpawnBurst_Instantaneous(Spawn Probability=0.25, Spawn Time=0.15, Use Spawn Probability=true, UseLoopCountLimit=false) {module=SpawnBurst_Instantaneous001}
  ]
  particleSpawnScript [
   &FN_SetSpawnGroup(Number of Spawn Groups=2)
   &InitializeParticle(
    Color=&MultiplyLinearColor(Linear Color=$User.Color, Scale Factor=&MakeVector4_FromFloat(Float=1.15))
    Lifetime=&RandomRangeFloat(Minimum=0.12, Maximum=0.15)
    Mesh Scale=&MaskVectorBySpawnGroup(Masked Vector=<1 1 1>, Passthrough Vector=<1 1 -1>)
    Mesh Scale Mode=::NewEnumerator4
    Sprite Size Mode=::NewEnumerator3
    Uniform Sprite Size=20
    UsePositionOffset=false
    Write Material Random=true
   )
   &SkeletalMeshLocation(
    Bone / Socket Index=1
    Bone Sampling Mode=::NewEnumerator4
    MaskSpawnGroup=false
    Sampled Orientation Coordinate Space=::Local
    Skeletal Mesh=$User.SkeletalMesh
   )
   &ShapeLocation(
    Additional Quaternion Rotation=true
    Additional Yaw / Pitch / Roll=<0 0.25 0>
    Cylinder Height=10
    Cylinder Height Midpoint=1
    Cylinder Radius=3.5
    MaskSpawnGroup=false
    Override Seed=false
    Rotation Mode=::NewEnumerator3
    Rotation Quaternion=$Particles.SkeletalMeshLocation.SampledRotation
    Shape Primitive=::NewEnumerator1
    Surface Only=true
    Surface Only Band Thickness=1
    UseOwnerScale=true
   )
   &InitialMeshOrientation(Mesh Orientation Mode=::NewEnumerator7, Use Rotation Vector=true)
   /SetVariables(Particles.VisibilityTag=&RandomRangeInt(Minimum=0, Maximum=2)) {isSetParameters=true, module=SetVariables_#1}
   /SetVariables(
    Particles.LightningSubImageIndex:NiagaraFloat=&UniformRangedFloat(Minimum=0, Maximum=3)
    Particles.SpriteSize=&Multiply_Vector2D_ByFloat(
     Float=&MaskFloatBySpawnGroup(Masked Float=&RandomRangeFloat(Minimum=1, Maximum=2), Passthrough Float=1, Spawn Group Mask=1)
     Vector2D=$Particles.SpriteSize
    )
   ) {isSetParameters=true, module=SetVariables_#2}
  ]
  particleUpdateScript [
   &ParticleState()
   &ScaleMeshSize(Scale Factor=&VectorFromFloat(Value=1))
   &ScaleColor(
    Scale Alpha=&FloatFromCurve(FloatCurve=DataInterface() {propertyValues {Curve {keys [<at 0.5 1 RCIM_Cubic>,<at 1 0 RCIM_Cubic>]}}})
    Scale RGB=&VectorFromFloat(Value=1)
    ScaleA=true
    ScaleRGB=true
   )
   &CameraOffset(Camera Offset Amount=10)
   /SetVariables(Particles.MeshOrientation=&MultiplyQuaternion(
     Quaternion A=$Particles.Initial.MeshOrientation
     Quaternion B=&MakeQuatFromAxisAngle(Axis=<1 0 0>, AngleInDegrees=&RandomRangeFloat(Minimum=0, Maximum=360))
    )) {isSetParameters=true, module=SetVariables_#3}
   &DynamicMaterialParameters(
    Index 0 Param 1=&FloatFromCurve(FloatCurve=DataInterface() {propertyValues {Curve {keys [<at 0 1>,<at 1 0>]}}})
    Index 0 Param 2=&FloatFromCurve(
     FloatCurve=DataInterface() {propertyValues {Curve {keys [
         <at 0 1 RCIM_Cubic>
         {arriveTangent=0.987, interpMode=RCIM_Cubic, leaveTangent=0.987, tangentMode=RCTM_User, time=0.4992, value=1}
         <at 0.571 1 RCIM_Cubic>
         <at 1 1>
        ]}}}
     Scale Curve=2500
    )
    Index 0 Param 3=&RandomRangeFloat(Minimum=1, Maximum=2)
    Index 0 Param 4=&RandomRangeFloat(Minimum=0, Maximum=0)
    Module.Param0WriteEnabled=true
    Module.Param1WriteEnabled=true
    Module.Param2WriteEnabled=true
    Module.Param3WriteEnabled=true
   ) {module=DynamicMaterialParameters004}
   &FN_Sprite_Attributes(
    Camera Offset=15
    Color=&MultiplyLinearColor(Linear Color=$Particles.Color, Scale Factor=<1 1 1 0.25>)
    Create Camera Offset Attribute=true
    Create Color Attribute=true
    Create Visibility Attribute=true
    Write to Module Namespace=true
   )
   &FN_Light_Attributes(
    Color=$Particles.Color
    Create Color Attribute=true
    Create Exponent Attribute=true
    Create Position Attribute=true
    Create Radius Attribute=true
    Enable Particle Light Enable=true
    Exponent=1
    Light Enabled:NiagaraBool=&SetBoolByIntComparison(
     A=&ReturnParticleIDAttributes(Particle ID Attribute=::NewEnumerator1)
     B=0
     Comparison Type=::NewEnumerator4
    )
    Position=&AddVectorToPosition(Position=$Particles.Position, Vector=<0 0 2.5>)
    Radius=20
    Scale Color=&FloatFromCurve(
     FloatCurve=DataInterface() {propertyValues {Curve {keys [
         <at 0 0 RCIM_Cubic>
         {arriveTangent=0.875, interpMode=RCIM_Cubic, leaveTangent=0.875, time=0.1, value=0.9}
         {arriveTangent=-0.2131, interpMode=RCIM_Cubic, leaveTangent=-0.2131, tangentMode=RCTM_User, time=0.4, value=0.35}
         {arriveTangent=-0.5833, interpMode=RCIM_Cubic, leaveTangent=-0.5833, time=0.7, value=0.5}
         <at 1 0 RCIM_Cubic>
        ]}}}
     Scale Curve=1.5
    )
   )
   /SetVariables(Particles.SpriteLightningDynamic:Vector4f=&MakeVector4(
     X=&RandomRangeFloat(Minimum=0, Maximum=1)
     Y=144
     Z=&FloatFromCurve(FloatCurve=DataInterface() {propertyValues {Curve {keys [
          {arriveTangent=-19.4071, interpMode=RCIM_Cubic, leaveTangent=-19.4071, tangentMode=RCTM_User, time=0, value=1}
          {interpMode=RCIM_Cubic, tangentMode=RCTM_Break, time=0.1, value=0}
          {arriveTangent=3.3801, interpMode=RCIM_Cubic, leaveTangent=3.3801, tangentMode=RCTM_User, time=1, value=1}
         ]}}})
     W=0
    )) {isSetParameters=true, module=SetVariables_#4}
   &FN_Sprite_Attributes(
    Alignment:Vector3f=$Particles.ShapeLocation.ShapeNormal
    Camera Offset=5
    Color=&MultiplyLinearColor(Linear Color=$Particles.Color, Scale Factor=&MakeVector4_FromFloat(Float=5))
    Create Alignment Attribute=true
    Create Camera Offset Attribute=true
    Create Color Attribute=true
    Create Sprite Rotation Attribute=true
    Create SpriteSize Attribute=true
    SpriteRotation:NiagaraFloat=&RandomRangeFloat(Minimum=0, Maximum=360)
    SpriteSize=&Multiply_Vector2D_ByFloat(Float=0.6, Vector2D=$Particles.SpriteSize)
    Write to Module Namespace=true
   ) {module=FN_Sprite_Attributes001}
  ]
  rendererClasses [
   &NiagaraMeshRendererProperties
   &NiagaraSpriteRendererProperties
   &NiagaraLightRendererProperties
  ]
  renderers [
   {rendererClass=&NiagaraMeshRendererProperties, rendererIndex=0}
   {rendererClass=&NiagaraSpriteRendererProperties, rendererIndex=1}
   {rendererClass=&NiagaraLightRendererProperties, rendererIndex=2}
   {rendererClass=&NiagaraMeshRendererProperties, rendererIndex=3}
   {rendererClass=&NiagaraSpriteRendererProperties, rendererIndex=4}
  ]
 }
 LightningMeshHands {
  emitterUpdateScript [
   &EmitterState()
   &SpawnBurst_Instantaneous(Use Spawn Probability=false, UseLoopCountLimit=false)
   &SpawnBurst_Instantaneous(Spawn Probability=0.25, Spawn Time=0.15, Use Spawn Probability=true, UseLoopCountLimit=false) {module=SpawnBurst_Instantaneous001}
  ]
  particleSpawnScript [
   &FN_SetSpawnGroup(Number of Spawn Groups=2)
   &InitializeParticle(
    Color=&MultiplyLinearColor(Linear Color=$User.Color, Scale Factor=&MakeVector4_FromFloat(Float=1.15))
    Lifetime=&RandomRangeFloat(Minimum=0.12, Maximum=0.15)
    Mesh Scale=&MaskVectorBySpawnGroup(Masked Vector=<1 1 1>, Passthrough Vector=<1 1 -1>)
    Mesh Scale Mode=::NewEnumerator4
    Sprite Size Mode=::NewEnumerator3
    Uniform Sprite Size=20
    UsePositionOffset=false
    Write Material Random=true
   )
   &SkeletalMeshLocation(
    Bone / Socket Index=2
    Bone Sampling Mode=::NewEnumerator4
    MaskSpawnGroup=false
    Sampled Orientation Coordinate Space=::Local
    Skeletal Mesh=$User.SkeletalMesh
   )
   &ShapeLocation(
    Additional Quaternion Rotation=true
    Additional Yaw / Pitch / Roll=<0 0.25 0>
    Cylinder Height=10
    Cylinder Height Midpoint=1
    Cylinder Radius=3.5
    MaskSpawnGroup=false
    Override Seed=false
    Rotation Mode=::NewEnumerator3
    Rotation Quaternion=$Particles.SkeletalMeshLocation.SampledRotation
    Shape Primitive=::NewEnumerator1
    Surface Only=true
    Surface Only Band Thickness=1
    UseOwnerScale=true
   )
   &InitialMeshOrientation(Mesh Orientation Mode=::NewEnumerator7, Use Rotation Vector=true)
   /SetVariables(Particles.VisibilityTag=&RandomRangeInt(Minimum=0, Maximum=1)) {isSetParameters=true, module=SetVariables_#1}
   /SetVariables(
    Particles.LightningSubImageIndex:NiagaraFloat=&UniformRangedFloat(Minimum=0, Maximum=3)
    Particles.SpriteSize=&Multiply_Vector2D_ByFloat(
     Float=&MaskFloatBySpawnGroup(Masked Float=&RandomRangeFloat(Minimum=1, Maximum=2), Passthrough Float=1, Spawn Group Mask=1)
     Vector2D=$Particles.SpriteSize
    )
   ) {isSetParameters=true, module=SetVariables_#2}
  ]
  particleUpdateScript [
   &ParticleState()
   &ScaleMeshSize(Scale Factor=&VectorFromFloat(Value=0.75))
   &ScaleColor(
    Scale Alpha=&FloatFromCurve(FloatCurve=DataInterface() {propertyValues {Curve {keys [<at 0.5 1 RCIM_Cubic>,<at 1 0 RCIM_Cubic>]}}})
    Scale RGB=&VectorFromFloat(Value=1)
    ScaleA=true
    ScaleRGB=true
   )
   &CameraOffset(Camera Offset Amount=10)
   /SetVariables(Particles.MeshOrientation=&MultiplyQuaternion(
     Quaternion A=$Particles.Initial.MeshOrientation
     Quaternion B=&MakeQuatFromAxisAngle(Axis=<1 0 0>, AngleInDegrees=&RandomRangeFloat(Minimum=0, Maximum=360))
    )) {isSetParameters=true, module=SetVariables_#3}
   &DynamicMaterialParameters(
    Index 0 Param 1=&FloatFromCurve(FloatCurve=DataInterface() {propertyValues {Curve {keys [<at 0 1>,<at 1 0>]}}})
    Index 0 Param 2=&FloatFromCurve(
     FloatCurve=DataInterface() {propertyValues {Curve {keys [
         <at 0 1 RCIM_Cubic>
         {arriveTangent=0.987, interpMode=RCIM_Cubic, leaveTangent=0.987, tangentMode=RCTM_User, time=0.4992, value=1}
         <at 0.571 1 RCIM_Cubic>
         <at 1 1>
        ]}}}
     Scale Curve=2500
    )
    Index 0 Param 3=&RandomRangeFloat(Minimum=1, Maximum=2)
    Index 0 Param 4=&RandomRangeFloat(Minimum=0, Maximum=0)
    Module.Param0WriteEnabled=true
    Module.Param1WriteEnabled=true
    Module.Param2WriteEnabled=true
    Module.Param3WriteEnabled=true
   ) {module=DynamicMaterialParameters004}
   &FN_Sprite_Attributes(
    Camera Offset=15
    Color=&MultiplyLinearColor(Linear Color=$Particles.Color, Scale Factor=<1 1 1 0.25>)
    Create Camera Offset Attribute=true
    Create Color Attribute=true
    Create Visibility Attribute=true
    Write to Module Namespace=true
   )
   &FN_Light_Attributes(
    Color=$Particles.Color
    Create Color Attribute=true
    Create Exponent Attribute=true
    Create Position Attribute=true
    Create Radius Attribute=true
    Enable Particle Light Enable=true
    Exponent=1
    Light Enabled:NiagaraBool=&SetBoolByIntComparison(
     A=&ReturnParticleIDAttributes(Particle ID Attribute=::NewEnumerator1)
     B=0
     Comparison Type=::NewEnumerator4
    )
    Position=&AddVectorToPosition(Position=$Particles.Position, Vector=<0 0 2.5>)
    Radius=20
    Scale Color=&FloatFromCurve(
     FloatCurve=DataInterface() {propertyValues {Curve {keys [
         <at 0 0 RCIM_Cubic>
         {arriveTangent=0.875, interpMode=RCIM_Cubic, leaveTangent=0.875, time=0.1, value=0.9}
         {arriveTangent=-0.2131, interpMode=RCIM_Cubic, leaveTangent=-0.2131, tangentMode=RCTM_User, time=0.4, value=0.35}
         {arriveTangent=-0.5833, interpMode=RCIM_Cubic, leaveTangent=-0.5833, time=0.7, value=0.5}
         <at 1 0 RCIM_Cubic>
        ]}}}
     Scale Curve=1.5
    )
   )
   /SetVariables(Particles.SpriteLightningDynamic:Vector4f=&MakeVector4(
     X=&RandomRangeFloat(Minimum=0, Maximum=1)
     Y=144
     Z=&FloatFromCurve(FloatCurve=DataInterface() {propertyValues {Curve {keys [
          {arriveTangent=-19.4071, interpMode=RCIM_Cubic, leaveTangent=-19.4071, tangentMode=RCTM_User, time=0, value=1}
          {interpMode=RCIM_Cubic, tangentMode=RCTM_Break, time=0.1, value=0}
          {arriveTangent=3.3801, interpMode=RCIM_Cubic, leaveTangent=3.3801, tangentMode=RCTM_User, time=1, value=1}
         ]}}})
     W=0
    )) {isSetParameters=true, module=SetVariables_#4}
   &FN_Sprite_Attributes(
    Alignment:Vector3f=$Particles.ShapeLocation.ShapeNormal
    Camera Offset=5
    Color=&MultiplyLinearColor(Linear Color=$Particles.Color, Scale Factor=&MakeVector4_FromFloat(Float=5))
    Create Alignment Attribute=true
    Create Camera Offset Attribute=true
    Create Color Attribute=true
    Create Sprite Rotation Attribute=true
    Create SpriteSize Attribute=true
    SpriteRotation:NiagaraFloat=&RandomRangeFloat(Minimum=0, Maximum=360)
    SpriteSize=&Multiply_Vector2D_ByFloat(Float=0.4, Vector2D=$Particles.SpriteSize)
    Write to Module Namespace=true
   ) {module=FN_Sprite_Attributes001}
  ]
  rendererClasses [
   &NiagaraMeshRendererProperties
   &NiagaraSpriteRendererProperties
   &NiagaraLightRendererProperties
  ]
  renderers [
   {rendererClass=&NiagaraMeshRendererProperties, rendererIndex=0}
   {rendererClass=&NiagaraSpriteRendererProperties, rendererIndex=1}
   {rendererClass=&NiagaraLightRendererProperties, rendererIndex=2}
   {rendererClass=&NiagaraSpriteRendererProperties, rendererIndex=3}
  ]
 }
 LightningSplinesArms {
  emitterUpdateScript [
   &EmitterState()
   &SpawnBurst_Instantaneous(
    Spawn Count=&RandomRangeInt(Minimum=2, Maximum=3)
    Spawn Probability=0.25
    Use Spawn Probability=true
    UseLoopCountLimit=false
   )
   &SpawnBurst_Instantaneous(Spawn Probability=0.25, Spawn Time=0.15, Use Spawn Probability=true, UseLoopCountLimit=false) {module=SpawnBurst_Instantaneous001}
  ]
  particleSpawnScript [
   &InitializeParticle(
    Color=&MultiplyLinearColor(Linear Color=$User.Color, Scale Factor=&MakeVector4_FromFloat(Float=222))
    Lifetime=&RandomRangeFloat(Minimum=0.17, Maximum=0.3)
    Mesh Scale Mode=::NewEnumerator3
    UsePositionOffset=false
    Write Material Random=true
   )
   &DynamicMaterialParameters(
    Index 0 Param 1=2
    Index 0 Param 2=&RandomRangeFloat(Minimum=0.65, Maximum=0.75)
    Index 0 Param 4=0.25
    Module.Param0WriteEnabled=true
    Module.Param1WriteEnabled=true
    Module.Param2WriteEnabled=true
    Module.Param3WriteEnabled=true
   )
   /SetVariables(Particles.MeshIndex=&RandomRangeInt(Minimum=0, Maximum=3)) {isSetParameters=true, module=SetVariables_#5}
  ]
  particleUpdateScript [
   &ParticleState()
   &SkeletalMeshLocation(Bone / Socket Index=1, Bone Sampling Mode=::NewEnumerator4, MaskSpawnGroup=false, Skeletal Mesh=$User.SkeletalMesh)
   &ScaleMeshSize(Scale Factor=&VectorFromFloat(Value=1))
   &ScaleColor(
    Scale Alpha=&FloatFromCurve(FloatCurve=DataInterface() {propertyValues {Curve {keys [<at 0.5 1 RCIM_Cubic>,<at 1 0 RCIM_Cubic>]}}}, Scale Curve=2)
    ScaleA=true
    ScaleRGB=true
   )
   &DynamicMaterialParameters(
    Index 0 Param 1=&RandomRangeFloat(Minimum=1.25, Maximum=1.75)
    Index 0 Param 3=$Particles.NormalizedAge
    Index 0 Param 4=0.25
    Module.Param0WriteEnabled=true
    Module.Param1WriteEnabled=false
    Module.Param2WriteEnabled=true
    Module.Param3WriteEnabled=true
   ) {module=DynamicMaterialParameters001}
   &FN_Sprite_Attributes(
    Camera Offset=10
    Color=&MultiplyLinearColor(
     Linear Color=$Particles.Color
     Scale Factor=&MakeFloat4FromVectorAndFloat(
      "Vector (RGB)"=<1 1 1>
      "Float (Alpha)"=&FloatFromCurve(
       FloatCurve=DataInterface() {propertyValues {Curve {keys [{interpMode=RCIM_Cubic, tangentMode=RCTM_User, time=0, value=1}, {interpMode=RCIM_Cubic, tangentMode=RCTM_User, time=1, value=0}]}}}
       Scale Curve=0
      )
     )
    )
    Create Camera Offset Attribute=true
    Create Color Attribute=true
    Create SpriteSize Attribute=true
    SpriteSize=&Multiply_Vector2D_ByFloat(Float=20, Vector2D=<1 1>)
    Write to Module Namespace=true
   )
   &FN_Light_Attributes(
    Color=$Particles.Color
    Create Color Attribute=true
    Create Exponent Attribute=true
    Create Position Attribute=true
    Create Radius Attribute=true
    Enable Particle Light Enable=true
    Exponent=1
    Light Enabled=true
    Position=&AddVectorToPosition(Position=$Particles.Position, Vector=<0 0 10>)
    Radius=20
    Scale Color=&FloatFromCurve(
     FloatCurve=DataInterface() {propertyValues {Curve {keys [
         <at 0 0 RCIM_Cubic>
         {arriveTangent=0.875, interpMode=RCIM_Cubic, leaveTangent=0.875, time=0.1, value=0.9}
         {arriveTangent=-0.2131, interpMode=RCIM_Cubic, leaveTangent=-0.2131, tangentMode=RCTM_User, time=0.4, value=0.35}
         {arriveTangent=-0.5833, interpMode=RCIM_Cubic, leaveTangent=-0.5833, time=0.7, value=0.5}
         <at 1 0 RCIM_Cubic>
        ]}}}
     Scale Curve=0.01
    )
   )
  ]
  rendererClasses [
   &NiagaraMeshRendererProperties
   &NiagaraSpriteRendererProperties
   &NiagaraLightRendererProperties
  ]
  renderers [
   {rendererClass=&NiagaraMeshRendererProperties, rendererIndex=0}
   {rendererClass=&NiagaraSpriteRendererProperties, rendererIndex=1}
   {rendererClass=&NiagaraLightRendererProperties, rendererIndex=2}
  ]
 }
 LightningSplinesHands {
  emitterUpdateScript [
   &EmitterState()
   &SpawnBurst_Instantaneous(
    Spawn Count=&RandomRangeInt(Minimum=2, Maximum=3)
    Spawn Probability=0.25
    Use Spawn Probability=true
    UseLoopCountLimit=false
   )
   &SpawnBurst_Instantaneous(Spawn Probability=0.25, Spawn Time=0.15, Use Spawn Probability=true, UseLoopCountLimit=false) {module=SpawnBurst_Instantaneous001}
  ]
  particleSpawnScript [
   &InitializeParticle(
    Color=&MultiplyLinearColor(Linear Color=$User.Color, Scale Factor=&MakeVector4_FromFloat(Float=222))
    Lifetime=&RandomRangeFloat(Minimum=0.17, Maximum=0.3)
    Mesh Scale Mode=::NewEnumerator3
    UsePositionOffset=false
    Write Material Random=true
   )
   &DynamicMaterialParameters(
    Index 0 Param 1=2
    Index 0 Param 2=&RandomRangeFloat(Minimum=0.65, Maximum=0.75)
    Index 0 Param 4=0.5
    Module.Param0WriteEnabled=true
    Module.Param1WriteEnabled=true
    Module.Param2WriteEnabled=true
    Module.Param3WriteEnabled=true
   )
   /SetVariables(Particles.MeshIndex=&RandomRangeInt(Minimum=0, Maximum=2)) {isSetParameters=true, module=SetVariables_#5}
  ]
  particleUpdateScript [
   &ParticleState()
   &SkeletalMeshLocation(Bone / Socket Index=2, Bone Sampling Mode=::NewEnumerator4, MaskSpawnGroup=false, Skeletal Mesh=$User.SkeletalMesh)
   &ScaleMeshSize(Scale Factor=&VectorFromFloat(Value=1))
   &ScaleColor(
    Scale Alpha=&FloatFromCurve(FloatCurve=DataInterface() {propertyValues {Curve {keys [<at 0.5 1 RCIM_Cubic>,<at 1 0 RCIM_Cubic>]}}}, Scale Curve=2)
    Scale RGB=&VectorFromFloat(Value=1)
    ScaleA=true
    ScaleRGB=true
   )
   &DynamicMaterialParameters(
    Index 0 Param 1=&RandomRangeFloat(Minimum=1.25, Maximum=1.75)
    Index 0 Param 3=$Particles.NormalizedAge
    Index 0 Param 4=0.25
    Module.Param0WriteEnabled=true
    Module.Param1WriteEnabled=false
    Module.Param2WriteEnabled=true
    Module.Param3WriteEnabled=true
   ) {module=DynamicMaterialParameters001}
   &FN_Sprite_Attributes(
    Camera Offset=10
    Color=&MultiplyLinearColor(
     Linear Color=$Particles.Color
     Scale Factor=&MakeFloat4FromVectorAndFloat(
      "Vector (RGB)"=<1 1 1>
      "Float (Alpha)"=&FloatFromCurve(
       FloatCurve=DataInterface() {propertyValues {Curve {keys [{interpMode=RCIM_Cubic, tangentMode=RCTM_User, time=0, value=1}, {interpMode=RCIM_Cubic, tangentMode=RCTM_User, time=1, value=0}]}}}
       Scale Curve=0
      )
     )
    )
    Create Camera Offset Attribute=true
    Create Color Attribute=true
    Create SpriteSize Attribute=true
    SpriteSize=&Multiply_Vector2D_ByFloat(Float=20, Vector2D=<1 1>)
    Write to Module Namespace=true
   )
   &FN_Light_Attributes(
    Color=$Particles.Color
    Create Color Attribute=true
    Create Exponent Attribute=true
    Create Position Attribute=true
    Create Radius Attribute=true
    Enable Particle Light Enable=true
    Exponent=1
    Light Enabled=true
    Position=&AddVectorToPosition(Position=$Particles.Position, Vector=<0 0 10>)
    Radius=20
    Scale Color=&FloatFromCurve(
     FloatCurve=DataInterface() {propertyValues {Curve {keys [
         <at 0 0 RCIM_Cubic>
         {arriveTangent=0.875, interpMode=RCIM_Cubic, leaveTangent=0.875, time=0.1, value=0.9}
         {arriveTangent=-0.2131, interpMode=RCIM_Cubic, leaveTangent=-0.2131, tangentMode=RCTM_User, time=0.4, value=0.35}
         {arriveTangent=-0.5833, interpMode=RCIM_Cubic, leaveTangent=-0.5833, time=0.7, value=0.5}
         <at 1 0 RCIM_Cubic>
        ]}}}
     Scale Curve=0.01
    )
   )
  ]
  rendererClasses [
   &NiagaraMeshRendererProperties
   &NiagaraSpriteRendererProperties
   &NiagaraLightRendererProperties
  ]
  renderers [
   {rendererClass=&NiagaraMeshRendererProperties, rendererIndex=0}
   {rendererClass=&NiagaraSpriteRendererProperties, rendererIndex=1}
   {rendererClass=&NiagaraLightRendererProperties, rendererIndex=2}
  ]
 }
}
summary {
 systemName=NS_Emote_Sahara_RandomLightning
 emitters [
  {
   emitterName=LightningMeshArms
   rendererClasses [
    &NiagaraMeshRendererProperties
    &NiagaraSpriteRendererProperties
    &NiagaraLightRendererProperties
   ]
  }
  {
   emitterName=LightningSplinesArms
   rendererClasses [
    &NiagaraMeshRendererProperties
    &NiagaraSpriteRendererProperties
    &NiagaraLightRendererProperties
   ]
  }
  {
   emitterName=LightningMeshHands
   rendererClasses [
    &NiagaraMeshRendererProperties
    &NiagaraSpriteRendererProperties
    &NiagaraLightRendererProperties
   ]
  }
  {
   emitterName=LightningSplinesHands
   rendererClasses [
    &NiagaraMeshRendererProperties
    &NiagaraSpriteRendererProperties
    &NiagaraLightRendererProperties
   ]
  }
 ]
 userVariables [
  User.Color {type=LinearColor, description="", defaultValue {struct=LinearColor, value=<rgba 0.25 0.6 1 1>}}
  User.SkeletalMesh {
   type=NiagaraDataInterfaceSkeletalMesh
   defaultValue {
    struct=NiagaraExt_VariableValue_DataInterface
    value {dataInterface=&self:NiagaraDataInterfaceSkeletalMesh_0}
   }
  }
 ]
}
