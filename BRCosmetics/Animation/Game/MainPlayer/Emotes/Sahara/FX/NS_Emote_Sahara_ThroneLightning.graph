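# Niagara one-shot lightning impact effect for the Sahara emote's throne, built from short burst emitters.
# Staggered ImpactLightning meshes and sprites flicker via curves, followed by delayed residue arcs, while a flare with a light and gravity-pulled sparks add the hit.
# Up-facing sprites form ground decal, airglow and smoke, with User.bIsInWater suppressing the smoke.
# User.Color tints most layers, ImpactLightning uses fixed colors, and LightningMesh_01 is disabled.
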NiagaraSystem /BRCosmetics/Animation/Game/MainPlayer/Emotes/Sahara/FX/NS_Emote_Sahara_ThroneLightning
dependencies {
 usedDataInterfaces [&NiagaraDataInterfaceCurve, &NiagaraDataInterfaceVector2DCurve]
 usedDynamicInputs []
 usedModules [&SystemState]
}
emitters {
 Blue_Decal {
  emitterUpdateScript [
   &EmitterState()
   &SpawnBurst_Instantaneous(Spawn Count=1, Use Spawn Probability=false)
  ]
  particleSpawnScript [
   /Niagara/Modules/Spawn/Initialization/InitializeParticle(
    Color=&MultiplyLinearColor(Linear Color=$User.Color, Scale Factor=&MakeVector4_FromFloat(Float=0.5))
    Lifetime=0.5
    Module.Write Color=true
    Module.Write Lifetime=true
    Module.Write Mass=true
    Module.Write Material Random=true
    Module.Write Position=false
    Module.Write Scale=false
    Module.Write SpriteRotation=false
    Module.Write SpriteSize=true
    Position=&SimulationPosition()
    Sprite Size=&Vector2DFromFloat(Value=120)
   )
   &SpriteFacingAndAlignment(Sprite Facing=<0 0 1>, WriteAlignment=false, WriteFacing=true)
  ]
  particleUpdateScript [
   &ParticleState()
   &SolveForcesAndVelocity(Clamp Velocity=false, Limit Acceleration=false)
   &ScaleColor(
    Scale Alpha=&FloatFromCurve(FloatCurve=DataInterface() {propertyValues {Curve {keys [
         {arriveTangent=9.0725, interpMode=RCIM_Cubic, leaveTangent=9.0725, tangentMode=RCTM_User, time=0, value=0}
         {arriveTangent=-2.5249, interpMode=RCIM_Cubic, leaveTangent=-2.5249, tangentMode=RCTM_User, time=0.1408, value=0.9847}
         {arriveTangent=-0.0405, interpMode=RCIM_Cubic, leaveTangent=-0.0405, tangentMode=RCTM_User, time=1, value=0}
        ]}}})
    Scale RGB=&VectorFromFloat(Value=20)
    ScaleA=true
    ScaleRGB=true
   )
  ]
  rendererClasses [&NiagaraSpriteRendererProperties]
  renderers [{rendererClass=&NiagaraSpriteRendererProperties}]
 }
 Flare {
  emitterUpdateScript [
   &EmitterState()
   &SpawnBurst_Instantaneous(Spawn Count=1, Use Spawn Probability=false)
  ]
  particleSpawnScript [
   &InitializeParticle(
    Color=&MultiplyLinearColor(Linear Color=$User.Color, Scale Factor=&MakeVector4_FromFloat(Float=3))
    Lifetime=0.5
    Module.Write Color=true
    Module.Write Lifetime=true
    Module.Write Mass=true
    Module.Write Material Random=true
    Module.Write Position=true
    Module.Write Scale=false
    Module.Write SpriteRotation=false
    Module.Write SpriteSize=true
    Position=&AddVectorToPosition(Position=/Niagara/DynamicInputs/Helpers/SimulationPosition(), Vector=<0 0 20>)
    Sprite Size=<700 700>
   )
   &AddVelocity(Coordinate Space=::World, Velocity=<0 0 0.1>)
   /Niagara/Modules/Spawn/Location/SphereLocation(MaskSpawnGroup=false, Module.Override Randomness=false, Sphere Radius=0, Surface Only=false)
  ]
  particleUpdateScript [
   &ParticleState()
   &ScaleSpriteSize(Scale Sprite Size Mode=::NewEnumerator0, Uniform Scale Factor=0.8)
   &SolveForcesAndVelocity(Clamp Velocity=false, Limit Acceleration=false)
   &ScaleColor(
    Scale Alpha=&FloatFromCurve(
     FloatCurve=DataInterface() {propertyValues {Curve {keys [
         {arriveTangent=-2.7664, interpMode=RCIM_Cubic, leaveTangent=-2.7664, tangentMode=RCTM_User, time=0, value=0.5}
         {arriveTangent=2, interpMode=RCIM_Cubic, leaveTangent=2, tangentMode=RCTM_Break, time=0.2, value=0.2621}
         {arriveTangent=-0.3277, interpMode=RCIM_Cubic, leaveTangent=-3.899, tangentMode=RCTM_Break, time=0.25, value=1}
         <at 1 0>
        ]}}}
     Scale Curve=2
    )
    ScaleA=true
    ScaleRGB=true
   )
   &DynamicMaterialParameters(
    Index 0 Param 1=&FloatFromCurve(FloatCurve=DataInterface() {propertyValues {Curve {keys [<at 0.25 0.7>,<at 0.5 0.05>]}}})
    Index 0 Param 2=2
    Index 0 Param 3=1
    Index 0 Param 4=1
    Module.Param0WriteEnabled=true
    Module.Param1WriteEnabled=true
    Module.Param2WriteEnabled=true
    Module.Param3WriteEnabled=true
   )
   &CameraOffset()
   &FN_Light_Attributes(
    Color=$Particles.Color
    Create Color Attribute=true
    Create Exponent Attribute=true
    Create Radius Attribute=true
    Exponent=1.5
    Radius=150
    Scale Color=&FloatFromCurve(
     FloatCurve=DataInterface() {propertyValues {Curve {keys [
         {arriveTangent=9.0725, interpMode=RCIM_Cubic, leaveTangent=9.0725, tangentMode=RCTM_User, time=0, value=0}
         {arriveTangent=-2.5249, interpMode=RCIM_Cubic, leaveTangent=-2.5249, tangentMode=RCTM_User, time=0.1408, value=0.9847}
         {arriveTangent=-0.0405, interpMode=RCIM_Cubic, leaveTangent=-0.0405, tangentMode=RCTM_User, time=1, value=0}
        ]}}}
     Scale Curve=5
    )
   )
  ]
  rendererClasses [&NiagaraSpriteRendererProperties, &NiagaraLightRendererProperties]
  renderers [
   {rendererClass=&NiagaraSpriteRendererProperties}
   {rendererClass=&NiagaraLightRendererProperties, rendererIndex=1}
  ]
 }
 GroundAirglow {
  emitterUpdateScript [
   &EmitterState()
   &SpawnBurst_Instantaneous(Spawn Count=2, Use Spawn Probability=false)
  ]
  particleSpawnScript [
   &FN_SetSpawnGroup(Particles Per Group=1, Pattern=::NewEnumerator0)
   &InitializeParticle(
    Color=&MultiplyLinearColor(Linear Color=$User.Color, Scale Factor=&MakeVector4_FromFloat(Float=0.5))
    Lifetime=0.7
    Module.Write Color=true
    Module.Write Lifetime=true
    Module.Write Mass=true
    Module.Write Material Random=true
    Module.Write Position=false
    Module.Write Scale=false
    Module.Write SpriteRotation=false
    Module.Write SpriteSize=true
    Position=/Niagara/Functions/Localspace/SimulationPosition()
    Sprite Size=&Vector2DFromFloat(Value=&Multiply_Float(A=400, B=&MaskFloatBySpawnGroup(Masked Float=1, Passthrough Float=0.8)))
   )
   &SpriteFacingAndAlignment(Sprite Facing=<0 0 1>, WriteAlignment=false, WriteFacing=true)
   /SetVariables(Particles.VisibilityTag=$Particles.SpawnGroup) {isSetParameters=true}
   &CameraOffset(Camera Offset Amount=&MaskFloatBySpawnGroup(Masked Float=0, Passthrough Float=100))
  ]
  particleUpdateScript [
   &ParticleState()
   &SolveForcesAndVelocity(Clamp Velocity=false, Limit Acceleration=false)
   &ScaleColor(
    Scale Alpha=&FloatFromCurve(
     FloatCurve=DataInterface() {propertyValues {Curve {keys [
         {arriveTangent=9.0725, interpMode=RCIM_Cubic, leaveTangent=9.0725, tangentMode=RCTM_User, time=0, value=0}
         {arriveTangent=-0.2874, interpMode=RCIM_Cubic, leaveTangent=-0.2874, tangentMode=RCTM_User, time=0.1408, value=0.9847}
         {arriveTangent=-0.0405, interpMode=RCIM_Cubic, leaveTangent=-0.0405, tangentMode=RCTM_User, time=1, value=0}
        ]}}}
     Scale Curve=2
    )
    Scale RGB=&VectorFromFloat(Value=&MaskFloatBySpawnGroup(Masked Float=2.5, Passthrough Float=3.5))
    ScaleA=true
    ScaleRGB=true
   )
  ]
  rendererClasses [&NiagaraSpriteRendererProperties]
  renderers [
   {rendererClass=&NiagaraSpriteRendererProperties}
   {rendererClass=&NiagaraSpriteRendererProperties, rendererIndex=1}
  ]
 }
 ImpactLightning {
  emitterUpdateScript [
   &EmitterState()
   &SpawnBurst_Instantaneous(Use Spawn Probability=false, UseLoopCountLimit=false)
   &SpawnBurst_Instantaneous(Spawn Time=0.0542, Use Spawn Probability=false, UseLoopCountLimit=false) {module=SpawnBurst_Instantaneous001}
   &SpawnBurst_Instantaneous(Spawn Time=0.0958, Use Spawn Probability=false, UseLoopCountLimit=false) {module=SpawnBurst_Instantaneous002}
  ]
  particleSpawnScript [
   /SetVariables(Particles.BiasedRandom=&FloatFromCurve(
     FloatCurve=DataInterface() {propertyValues {Curve {keys [
         {interpMode=RCIM_Cubic, tangentMode=RCTM_User, time=0, value=0}
         {arriveTangent=2.5197, interpMode=RCIM_Cubic, leaveTangent=2.5197, tangentMode=RCTM_User, time=0.6, value=0.6}
         {interpMode=RCIM_Cubic, tangentMode=RCTM_User, time=1, value=1}
        ]}}}
     CurveIndex=&RandomRangeFloat(Minimum=0, Maximum=1)
    )) {isSetParameters=true}
   &InitializeParticle(
    Color=<rgba 0.5 0.5 0.5 1>
    Lifetime=&Lerp_Float(A=0.2, B=0.25, Alpha=$Particles.BiasedRandom)
    Mesh Scale=<0.275 0.275 0.275>
    Module.Write Color=true
    Module.Write Lifetime=true
    Module.Write Mass=true
    Module.Write Material Random=true
    Module.Write Position=false
    Module.Write Scale=true
    Module.Write SpriteRotation=false
    Module.Write SpriteSize=true
    Position=&SimulationPosition()
    Sprite Size=&Vector2DFromFloat(Value=&RandomRangeFloat(Minimum=350, Maximum=450))
   )
   &InitialMeshOrientation(
    Mesh Axis To Orient=<0 0 1>
    Mesh Orientation Mode=::NewEnumerator2
    Orientation Vector=<-1 0 0>
    Rotation=&RandomRangeVector(Minimum=<0 0 0>, Maximum=<1 0 0>)
    Use Rotation Vector=true
   ) {module=InitialMeshOrientation001}
   /SetVariables(Particles.New Position:NiagaraPosition=<0 0 275>) {isSetParameters=true}
   /SetVariables(Particles.MeshIndex=&RandomRangeInt(Minimum=0, Maximum=1)) {isSetParameters=true}
   /SetVariables(Particles.DynParamsSprite:Vector4f=<0.25 2 1 1>) {isSetParameters=true}
  ]
  particleUpdateScript [
   &ParticleState()
   &ScaleColor(
    Scale Alpha=&FloatFromCurve(FloatCurve=DataInterface() {propertyValues {Curve {keys [
         <at 0 0>
         {interpMode=RCIM_Cubic, tangentMode=RCTM_User, time=0.1258, value=0.1952}
         <at 0.341 0.0902>
         {interpMode=RCIM_Cubic, tangentMode=RCTM_User, time=0.496, value=0.2532}
         <at 1 0>
        ]}}})
    Scale RGB=&VectorFromFloat(Value=&FloatFromCurve(FloatCurve=DataInterface() {propertyValues {Curve {keys [
          {arriveTangent=3.5209, interpMode=RCIM_Cubic, leaveTangent=3.5209, tangentMode=RCTM_User, time=0, value=0.5}
          {arriveTangent=-20.2701, interpMode=RCIM_Cubic, leaveTangent=-20.2701, tangentMode=RCTM_User, time=0.45, value=5}
          <at 1 1>
         ]}}}))
    ScaleA=true
    ScaleRGB=true
   )
   /SetVariables(Particles.SpriteColor=&MultiplyLinearColor(
     Linear Color=$Particles.Color
     Scale Factor=&MakeVector4_FromLinearColor(LinearColor=&MultiplyLinearColor(
       Linear Color=&MultiplyLinearColor(Linear Color=<rgba 0.2 1 0.9 1>, Scale Factor=&MakeVector4_FromFloat(Float=1.5))
       Scale Factor=&MakeVector4_FromFloat(Float=&Multiply_Float(
         A=1
         B=&FloatFromCurve(FloatCurve=DataInterface() {propertyValues {Curve {keys [
              {interpMode=RCIM_Cubic, tangentMode=RCTM_User, time=0, value=1}
              {arriveTangentWeight=0.031, interpMode=RCIM_Cubic, leaveTangentWeight=0.0366, tangentMode=RCTM_User, time=0.0918, value=0.05}
              {arriveTangentWeight=0.16, interpMode=RCIM_Cubic, leaveTangent=18.7153, leaveTangentWeight=0.1479, tangentMode=RCTM_Break, time=0.18, value=0}
              {arriveTangent=-0.2815, interpMode=RCIM_Cubic, leaveTangent=-0.2815, tangentMode=RCTM_User, time=0.3, value=0.95}
              {arriveTangentWeight=0.0345, interpMode=RCIM_Cubic, leaveTangentWeight=0.0566, tangentMode=RCTM_User, time=0.4, value=0.1}
              {
               arriveTangentWeight=0.1589
               interpMode=RCIM_Cubic
               leaveTangent=29.6956
               leaveTangentWeight=0.1018
               tangentMode=RCTM_Break
               time=0.5
               value=0.05
              }
              {arriveTangent=0.512, interpMode=RCIM_Cubic, leaveTangent=0.512, tangentMode=RCTM_User, time=0.5641, value=0.8648}
              <at 1 0>
             ]}}})
        ))
      ))
    )) {isSetParameters=true}
   &DynamicMaterialParameters(
    Index 0 Param 1=&FloatFromCurve(FloatCurve=DataInterface() {propertyValues {Curve {keys [{tangentMode=RCTM_User, time=0, value=1}, {tangentMode=RCTM_User, time=1, value=0}]}}})
    Index 0 Param 2=&Add_Float(
     A=&FloatFromCurve(FloatCurve=DataInterface() {propertyValues {Curve {keys [
          {interpMode=RCIM_Cubic, leaveTangent=5.6571, tangentMode=RCTM_Break, time=0, value=0}
          {arriveTangent=0.0136, interpMode=RCIM_Cubic, leaveTangent=-5.3647, tangentMode=RCTM_Break, time=0.4833, value=1}
          <at 1 -0.1>
         ]}}})
     B=&Multiply_Float(A=$Particles.BiasedRandom, B=0.5)
    )
    Index 0 Param 3=&RandomRangeFloat(Minimum=0, Maximum=0.25)
    Index 1 Param 1=&Add_Float(A=$Particles.NormalizedAge, B=&RandomRangeFloat(Minimum=0, Maximum=1))
    Index 1 Param 2=&RandomRangeFloat(Minimum=-25, Maximum=25)
    Index 1 Param 3=3
    Module.Param0WriteEnabled=true
    Module.Param0WriteEnabled1=true
    Module.Param1WriteEnabled=true
    Module.Param1WriteEnabled1=true
    Module.Param2WriteEnabled=true
    Module.Param2WriteEnabled1=true
    Module.Param3WriteEnabled=false
    Module.Param3WriteEnabled1=false
    Write Parameter Index 1=true
   ) {module=DynamicMaterialParameters001}
   &ScaleSpriteSize(Uniform Curve Sprite Scale=DataInterface() {propertyValues {Curve {keys [
        {arriveTangent=21.5526, interpMode=RCIM_Cubic, time=0, value=0}
        <at 0.1 1 RCIM_Cubic>
        {arriveTangent=-1.8758, interpMode=RCIM_Cubic, leaveTangent=-1.8758, tangentMode=RCTM_User, time=1, value=0}
       ]}}})
   &ScaleColor(Scale RGB=&VectorFromFloat(Value=1), ScaleA=true, ScaleRGB=true) {module=ScaleColor001}
  ]
  rendererClasses [&NiagaraMeshRendererProperties, &NiagaraSpriteRendererProperties]
  renderers [
   {rendererClass=&NiagaraMeshRendererProperties, rendererIndex=0}
   {rendererClass=&NiagaraSpriteRendererProperties, rendererIndex=1}
  ]
 }
 LightningMesh_01 {
  bEnabled=false
  emitterUpdateScript [
   &EmitterState()
   &SpawnBurst_Instantaneous(Spawn Count=&RandomRangeInt(Minimum=2, Maximum=3), Use Spawn Probability=false)
   &SpawnBurst_Instantaneous(Spawn Count=&RandomRangeInt(Minimum=1, Maximum=2), Spawn Time=0.1, Use Spawn Probability=false) {module=SpawnBurst_Instantaneous001}
  ]
  particleSpawnScript [
   &InitializeParticle(
    Color=&MakeLinearColorFromVectorAndFloat("Vector (RGB)"=<1 1 1>, "Float (Alpha)"=0)
    Lifetime=&Multiply_Float(A=&UniformRangedFloat(Minimum=0.07, Maximum=0.12), B=1)
    Mesh Scale=&RandomRangeVector(Minimum=<0.17 0.17 0.18>, Maximum=<0.2 0.2 0.3>)
    Module.Write Color=true
    Module.Write Lifetime=true
    Module.Write Mass=true
    Module.Write Material Random=true
    Module.Write Position=false
    Module.Write Scale=true
    Module.Write SpriteRotation=false
    Module.Write SpriteSize=false
    Position=&SimulationPosition()
   )
   &AddVelocity(Coordinate Space=::World, Velocity=<0 0 1>)
   &InitialMeshOrientation(
    Mesh Axis To Orient=<0 0 1>
    Mesh Orientation Mode=::NewEnumerator2
    Rotation=&Add_Vector(A=&RandomRangeVector(Minimum=<0 0 0>, Maximum=<0 0 1>), B=<0 0 0>)
    Use Rotation Vector=true
   ) {module=InitialMeshOrientation001}
   &CameraOffset(Camera Offset Amount=150)
  ]
  particleUpdateScript [
   &DynamicMaterialParameters(
    Index 0 Param 1=&UniformRangedFloat(Minimum=1, Maximum=5)
    Index 0 Param 2=&UniformRangedFloat(Minimum=0, Maximum=1)
    Index 0 Param 3=1
    Index 0 Param 4=1
    Module.Param0WriteEnabled=true
    Module.Param1WriteEnabled=true
    Module.Param2WriteEnabled=true
    Module.Param3WriteEnabled=true
   )
   &ParticleState()
   &SolveForcesAndVelocity(Clamp Velocity=false, Limit Acceleration=false)
   &self:ScratchModule(
    MeshOrientation:Quat4f=$Particles.Initial.MeshOrientation
    Rotation:NiagaraFloat=&RandomRangeFloat(Minimum=-0.75, Maximum=0.75)
    RotationAxis:Vector3f=<0 1 0>
   ) {module=ScratchModule}
   &ScaleColor(
    Scale Alpha=&FloatFromCurve(
     FloatCurve=DataInterface() {propertyValues {Curve {keys [<at 0 1 RCIM_Constant>,<at 0.1189 0 RCIM_Constant>,<at 0.2516 0.997 RCIM_Constant>,<at 0.6735 -0.0069 RCIM_Constant>,<at 0.7649 1.0117 RCIM_Constant>,<at 1 0 RCIM_Constant>]}}}
     Scale Curve=&Multiply_Float(
      A=&FloatFromCurve(FloatCurve=DataInterface() {propertyValues {Curve {keys [<at 0 1 RCIM_Constant>,<at 0.1561 -0.0068 RCIM_Constant>,<at 0.3011 0.9877 RCIM_Constant>,<at 0.6424 0.0069 RCIM_Constant>,<at 0.7618 0.9877 RCIM_Constant>,<at 1 0 RCIM_Constant>]}}})
      B=&UniformRangedFloat(Minimum=0, Maximum=1)
     )
    )
    Scale RGB=&VectorFromFloat(Value=&UniformRangedFloat(Minimum=1, Maximum=5))
    ScaleA=true
    ScaleRGB=true
   ) {module=ScaleColor001}
  ]
  rendererClasses [&NiagaraMeshRendererProperties]
  renderers [{rendererClass=&NiagaraMeshRendererProperties, rendererIndex=0}]
 }
 Lightning_residue {
  emitterUpdateScript [
   &EmitterState()
   &SpawnBurst_Instantaneous(Spawn Count=&RandomRangeInt(Minimum=1, Maximum=2), Spawn Time=0.18, Use Spawn Probability=false)
   &SpawnBurst_Instantaneous(Spawn Count=&RandomRangeInt(Minimum=1, Maximum=2), Spawn Time=0.3, Use Spawn Probability=false) {module=SpawnBurst_Instantaneous001}
  ]
  particleSpawnScript [
   &InitializeParticle(
    Color=&MultiplyLinearColor(Linear Color=$User.Color, Scale Factor=&MakeVector4_FromFloat(Float=50))
    Lifetime=&UniformRangedFloat(Minimum=0.08, Maximum=0.1)
    Mesh Scale=&UniformRangedVector(Minimum=<3 3 2.4>, Maximum=<8 8 2.4>)
    Module.Write Color=true
    Module.Write Lifetime=true
    Module.Write Mass=true
    Module.Write Material Random=true
    Module.Write Position=true
    Module.Write Scale=true
    Module.Write SpriteRotation=false
    Module.Write SpriteSize=false
    Position=&SimulationPosition()
   )
   &DynamicMaterialParameters(
    Index 0 Param 1=&UniformRangedFloat(Minimum=5, Maximum=10)
    Index 0 Param 3=&UniformRangedFloat(Minimum=0.2, Maximum=1)
    Index 0 Param 4=&UniformRangedFloat(Minimum=0, Maximum=1)
    Module.Param0WriteEnabled=true
    Module.Param1WriteEnabled=false
    Module.Param2WriteEnabled=true
    Module.Param3WriteEnabled=true
   )
   &SphereLocation(MaskSpawnGroup=false, Module.Override Randomness=false, Sphere Radius=10, Surface Only=false)
   &AddVelocity(Coordinate Space=::World, Velocity=<0 0 1>)
   &InitialMeshOrientation(
    Mesh Orientation Mode=::NewEnumerator2
    Rotation=&RandomRangeVector(Minimum=<0 0 0>, Maximum=<1 0 0>)
    Use Rotation Vector=true
   )
   &CameraOffset(Camera Offset Amount=150)
  ]
  particleUpdateScript [
   &ParticleState()
   &ScaleMeshSize(Scale Factor=&VectorFromFloat(Value=0.5))
   &SolveForcesAndVelocity(Clamp Velocity=false, Limit Acceleration=false)
   &ScaleColor(
    Scale Alpha=&FloatFromCurve(
     FloatCurve=DataInterface() {propertyValues {Curve {keys [<at 0 1 RCIM_Constant>,<at 0.1189 0 RCIM_Constant>,<at 0.2516 0.997 RCIM_Constant>,<at 0.6735 -0.0069 RCIM_Constant>,<at 0.7649 1.0117 RCIM_Constant>,<at 1 0 RCIM_Constant>]}}}
     Scale Curve=&Multiply_Float(
      A=&FloatFromCurve(FloatCurve=DataInterface() {propertyValues {Curve {keys [<at 0 1 RCIM_Constant>,<at 0.1561 -0.0068 RCIM_Constant>,<at 0.3011 0.9877 RCIM_Constant>,<at 0.6424 0.0069 RCIM_Constant>,<at 0.7618 0.9877 RCIM_Constant>,<at 1 0 RCIM_Constant>]}}})
      B=&UniformRangedFloat(Minimum=1, Maximum=1)
     )
    )
    Scale RGB=&VectorFromFloat(Value=&UniformRangedFloat(Minimum=1, Maximum=5))
    ScaleA=true
    ScaleRGB=true
   ) {module=ScaleColor001}
   &DynamicMaterialParameters(
    Index 0 Param 2=&FloatFromCurve(FloatCurve=DataInterface() {propertyValues {Curve {keys [{interpMode=RCIM_Cubic, leaveTangent=-2.846, tangentMode=RCTM_Break, time=0, value=1}, <at 1 -0.25>]}}})
    Module.Param0WriteEnabled=false
    Module.Param1WriteEnabled=true
    Module.Param2WriteEnabled=false
    Module.Param3WriteEnabled=false
   ) {module=DynamicMaterialParameters001}
   &self:ScratchModule(
    MeshOrientation:Quat4f=$Particles.Initial.MeshOrientation
    Rotation:NiagaraFloat=&RandomRangeFloat(Minimum=-0.1, Maximum=0.1)
    RotationAxis:Vector3f=<0 1 0>
   ) {enabled=false, module=ScratchModule}
  ]
  rendererClasses [&NiagaraMeshRendererProperties]
  renderers [{rendererClass=&NiagaraMeshRendererProperties, rendererIndex=0}]
 }
 SmokeGround {
  emitterUpdateScript [
   &EmitterState()
   &SpawnBurst_Instantaneous(
    Spawn Count=1
    Spawn Probability=&MakeCustomFloatFromBool(Bool=$User.bIsInWater, True Float=0, False Float=1)
    Use Spawn Probability=true
   )
  ]
  particleSpawnScript [
   &InitializeParticle(
    Color=&MultiplyLinearColor(Linear Color=$User.Color, Scale Factor=&MakeVector4_FromFloat(Float=0.5))
    Lifetime=0.4
    Module.Write Color=true
    Module.Write Lifetime=true
    Module.Write Mass=true
    Module.Write Material Random=true
    Module.Write Position=false
    Module.Write Scale=false
    Module.Write SpriteRotation=true
    Module.Write SpriteSize=true
    Position=&SimulationPosition()
    Sprite Size=&Vector2DFromFloat(Value=350)
   )
   /SetVariables(Particles.SpriteFacing=<0 0 1>) {isSetParameters=true}
   &SystemLocation(MaskSpawnGroup=false, Offset=<0 0 5>)
  ]
  particleUpdateScript [
   &ParticleState()
   &ScaleSpriteSize(Scale Factor=&Vector2DFromFloat(Value=&FloatFromCurve(FloatCurve=DataInterface() {propertyValues {Curve {keys [
          {arriveTangent=6.0515, interpMode=RCIM_Cubic, leaveTangent=6.0515, tangentMode=RCTM_User, time=0, value=0}
          {arriveTangent=0.3165, interpMode=RCIM_Cubic, leaveTangent=0.3165, tangentMode=RCTM_User, time=1.0011, value=2}
         ]}}})))
   &DynamicMaterialParameters(
    Index 0 Param 1=1
    Index 0 Param 2=5
    Index 0 Param 3=1
    Index 0 Param 4=1
    Index 1 Param 1=&FloatFromCurve(FloatCurve=DataInterface() {propertyValues {Curve {keys [{arriveTangent=22.1134, interpMode=RCIM_Cubic, time=0.3, value=4}, {interpMode=RCIM_Cubic, leaveTangent=6.3587, time=1, value=12}]}}})
    Index 1 Param 2=&Lerp_Float(A=0.01, B=0.22, Alpha=$Particles.NormalizedAge)
    Index 1 Param 3=2
    Module.Param0WriteEnabled=true
    Module.Param0WriteEnabled1=true
    Module.Param1WriteEnabled=true
    Module.Param1WriteEnabled1=true
    Module.Param2WriteEnabled=true
    Module.Param2WriteEnabled1=true
    Module.Param3WriteEnabled=true
    Module.Param3WriteEnabled1=true
    Write Parameter Index 1=true
   )
   &ScaleColor(
    Scale Alpha=&FloatFromCurve(
     FloatCurve=DataInterface() {propertyValues {Curve {keys [
         <at 0.5 1 RCIM_Cubic>
         {arriveTangent=-3.6613, interpMode=RCIM_Cubic, leaveTangent=-3.6613, tangentMode=RCTM_User, time=0.9, value=0.5}
         {arriveTangent=-4.0971, interpMode=RCIM_Cubic, leaveTangent=-4.0971, tangentMode=RCTM_User, time=1, value=0}
        ]}}}
     Scale Curve=2.5
    )
    ScaleA=true
    ScaleRGB=false
   )
  ]
  rendererClasses [&NiagaraSpriteRendererProperties]
  renderers [{rendererClass=&NiagaraSpriteRendererProperties}]
 }
 Sparks {
  emitterUpdateScript [
   &EmitterState()
   &SpawnBurst_Instantaneous(Spawn Count=40, Use Spawn Probability=false)
   &SpawnBurst_Instantaneous(Spawn Count=90, Spawn Group=1, Use Spawn Probability=false) {module=SpawnBurst_Instantaneous001}
  ]
  particleSpawnScript [
   &InitializeParticle(
    Color=&MultiplyLinearColor(Linear Color=$User.Color, Scale Factor=&MakeVector4_FromFloat(Float=30))
    Lifetime=&UniformRangedFloat(Minimum=0.2, Maximum=0.5)
    Mesh Scale=<1.5 5 1>
    Module.Write Color=true
    Module.Write Lifetime=true
    Module.Write Mass=true
    Module.Write Material Random=true
    Module.Write Position=false
    Module.Write Scale=false
    Module.Write SpriteRotation=false
    Module.Write SpriteSize=true
    Position=&SimulationPosition()
    Sprite Size=&Multiply_Vector2D_ByFloat(
     Float=&MaskFloatBySpawnGroup(Masked Float=0.6, Passthrough Float=1, Spawn Group Mask=1)
     Vector2D=&UniformRangedVector2D(Minimum=<1 5>, Maximum=<1 4>)
    )
   )
   &AddVelocity(Coordinate Space=::World, Velocity=&UniformRangedVector(Minimum=<0 0 200>, Maximum=<1 1 500>))
   &SphereLocation(MaskSpawnGroup=false, Module.Override Randomness=false, Non Uniform Scale=<2.5 2.5 0.5>, Sphere Radius=5, Surface Only=true)
   &AddVelocity(Velocity=&RandomRangeVector(Minimum=<0 0 50>, Maximum=<0 0 150>)) {module=AddVelocity001}
   &AddVelocityFromPoint(
    Module.Use Falloff Distance=false
    Velocity Strength=&Multiply_Float(
     A=&MaskFloatBySpawnGroup(Masked Float=150, Passthrough Float=300, Spawn Group Mask=1)
     B=&RandomRangeFloat(Minimum=0.8, Maximum=1.2)
    )
   )
  ]
  particleUpdateScript [
   &GravityForce(Gravity=<0 0 -1500>)
   &ParticleState()
   &InheritVelocity(Module.Limit Speed=true, Velocity Limit=20)
   &SolveForcesAndVelocity(Clamp Velocity=false, Limit Acceleration=false)
   &ScaleColor(
    Scale Alpha=&FloatFromCurve(FloatCurve=DataInterface() {propertyValues {Curve {keys [
         {arriveTangent=-2.7664, interpMode=RCIM_Cubic, leaveTangent=-2.7664, tangentMode=RCTM_User, time=0, value=0.5}
         {arriveTangent=2, interpMode=RCIM_Cubic, leaveTangent=2, tangentMode=RCTM_Break, time=0.2, value=0.2621}
         {arriveTangent=-0.3277, interpMode=RCIM_Cubic, leaveTangent=-3.899, tangentMode=RCTM_Break, time=0.25, value=1}
         <at 1 0>
        ]}}})
    ScaleA=true
    ScaleRGB=true
   )
   &DynamicMaterialParameters(
    Index 0 Param 1=&FloatFromCurve(FloatCurve=DataInterface() {propertyValues {Curve {keys [<at 0.25 0.7>,<at 0.5 0.05>]}}})
    Index 0 Param 2=2
    Index 0 Param 3=1
    Index 0 Param 4=1
    Module.Param0WriteEnabled=true
    Module.Param1WriteEnabled=true
    Module.Param2WriteEnabled=true
    Module.Param3WriteEnabled=true
   )
   &CameraOffset(Camera Offset Amount=50)
   &ScaleSpriteSize(Scale Factor=&Vector2DFromCurve(
     Vector2Curve=DataInterface() {propertyValues {XCurve {keys [<at 0 1.5>,<at 0.3 1.262>,<at 1 1>]}, YCurve {keys [<at 0 1>,<at 0.3 1.1>,<at 1 0>]}}}
     ScaleCurve=<2 2>
    ))
   &ScaleSpriteSize(Scale Factor=<0.4 1>, Scale Sprite Size Mode=::NewEnumerator2) {module=ScaleSpriteSize001}
  ]
  rendererClasses [&NiagaraSpriteRendererProperties]
  renderers [{rendererClass=&NiagaraSpriteRendererProperties}]
 }
}
summary {
 systemName=NS_Emote_Sahara_ThroneLightning
 emitters [
  {bEnabled=false, emitterName=LightningMesh_01, rendererClasses [&NiagaraMeshRendererProperties]}
  {emitterName=Lightning_residue, rendererClasses [&NiagaraMeshRendererProperties]}
  {emitterName=Flare, rendererClasses [&NiagaraSpriteRendererProperties, &NiagaraLightRendererProperties]}
  {emitterName=Sparks, rendererClasses [&NiagaraSpriteRendererProperties]}
  {emitterName=Blue_Decal, rendererClasses [&NiagaraSpriteRendererProperties]}
  {emitterName=SmokeGround, rendererClasses [&NiagaraSpriteRendererProperties]}
  {emitterName=GroundAirglow, rendererClasses [&NiagaraSpriteRendererProperties]}
  {
   emitterName=ImpactLightning
   rendererClasses [&NiagaraMeshRendererProperties, &NiagaraSpriteRendererProperties]
  }
 ]
 userVariables [
  User.bIsInWater {type=NiagaraBool}
  User.Color {type=LinearColor, description="", defaultValue {struct=LinearColor, value=<rgba 0.25 0.6 1 1>}}
 ]
}
